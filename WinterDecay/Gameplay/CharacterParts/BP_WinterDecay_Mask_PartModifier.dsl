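; Character part modifier for the WinterDecay teleport mask, managing its mask and right-hand Niagara effects.
; It registers with the current B_WinterDecay_Teleport weapon and attaches the effects, with cooldown events toggling their state.
; Vehicles, emotes and prop or cloak tags switch effects off or hide them.
; Game-time stamps in custom primitive data drive mask material transitions, while a delayed toss sequence hides the mask.
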(event Custom|Mask%20NS
  (bind _returnvalue (CustomCharacter|GetAssociatedPlayerPawn))
  (bind _mask_equip_ns (Variables|Default|GetMask%20Equip%20NS))
  (bind _ns_voidmask_handvoid (Variables|Default|GetNS_VoidMask_HandVoid))
  ;[0.4383 0.1161 1] TODO
  ;;[1 1 1] This is a work around for FORT-589563. A systemic fix described in FORT-617814 should make the original code work without log spam. Note also this work around required the addition of the Glove Equip NS component.
  (if true
    (Components|Activation|Activate _mask_equip_ns false)
    (Utilities|FlowControl|Sequence
      (:then_0
        ;;[0.4588 0.6885 0.8125 0.5] FX - Hide Orb if FX spawns when inside a vehicle (like switching seats)
        (if (Vehicle|IsinVehicle (CustomCharacter|GetAssociatedPlayerPawn))
          (Niagara|SetNiagaraVariable(Bool) _ns_voidmask_handvoid "HideOrb" true)
          (else
            (Niagara|SetNiagaraVariable(Bool) _ns_voidmask_handvoid "HideOrb" false))))
      (:then_1
        (Transformation|AttachComponentToComponent _mask_equip_ns (CustomCharacter|GetSkeletalMeshForPartOnPlayerPawn) "main" "KeepRelative" "KeepRelative" "KeepRelative" false)
        ;[1 0 0 0.5] VFX: Void on hand during equipping
        (Utilities|IsValid _returnvalue
          (:Is%20Valid
            (Transformation|AttachComponentToComponent (Variables|Default|GetNS_VoidMask_HandVoid) (Variables|/Script/Engine.Character|GetMesh _returnvalue) "hand_pickaxe_r" "KeepRelative" "KeepRelative" "KeepRelative" false)))))))

(event EventOnPostResIn
  (bind _returnvalue (CustomCharacter|GetAssociatedPlayerPawn))
  (bind _prophidetag (Variables|Default|GetPropHideTag))
  (|CallParentFunction)
  (Utilities|FlowControl|Sequence
    (:then_0
      (bind _asfort_player_pawn_athena (Utilities|Casting|CastToFortPlayerPawnAthena _returnvalue))
      (bind _asb_winter_decay_teleport (Utilities|Casting|CastToB_WinterDecay_Teleport (Variables|/Script/FortniteGame.FortPawn|GetCurrentWeapon _asfort_player_pawn_athena)))
      (bind _output_get (Variables|Default|SetWinterDecay _asb_winter_decay_teleport))
      (Variables|/WinterDecay/Gameplay/Actors/B_WinterDecay_Teleport.B_WinterDecay_Teleport_C|SetPartModifier self _output_get)
      (CallFunction|MaskNS)
      ; Hide Effects while in Prop
      (Ability|Tasks|WaitGameplayTagAddToActor _asfort_player_pawn_athena _prophidetag
        (:then
          (Ability|Tasks|WaitGameplayTagRemoveFromActor _asfort_player_pawn_athena _prophidetag
            (:then
              ; Turn invisible if the player goes invisible
              (Ability|Tasks|WaitGameplayTagAddToActor _asfort_player_pawn_athena (Variables|Default|GetCloakTag)
                (:then
                  (Ability|Tasks|WaitGameplayTagRemoveFromActor _asfort_player_pawn_athena (Variables|Default|GetCloakTag)
                    (:then
                      (Cosmetics|WhenCosmeticPartApplied _returnvalue (delegate Custom|CustomEvent) false)
                      (if (Utilities|Enum|Equal(Enum) (stale Custom|CustomEvent PartType) "Gameplay")
                        _))
                    (:Removed
                      (Rendering|SetVisibility (Variables|Default|GetNS_VoidMask_HandVoid) true true)
                      (Rendering|SetVisibility (Variables|Default|GetMask%20Equip%20NS) true true))))
                (:Added
                  (Rendering|SetVisibility (Variables|Default|GetNS_VoidMask_HandVoid) false true)
                  (Rendering|SetVisibility (Variables|Default|GetMask%20Equip%20NS) false true))))
            (:Removed
              (Rendering|SetVisibility (Variables|Default|GetNS_VoidMask_HandVoid) true true)
              (Rendering|SetVisibility (Variables|Default|GetMask%20Equip%20NS) true true))))
        (:Added
          (Rendering|SetVisibility (Variables|Default|GetNS_VoidMask_HandVoid) false true)
          (Rendering|SetVisibility (Variables|Default|GetMask%20Equip%20NS) false true))))
    (:then_1
      (Ability|Tasks|WaitGameplayTagAddToActor (CustomCharacter|GetAssociatedPlayerPawn) (Variables|Default|GetWearingTag)
        (:Added
          (Development|PrintString "AAAAA" true true "(R=1.000000,G=0.960122,B=0.217016,A=1.000000)"))))))

(event Custom|CustomEvent
  (params (Pawn :type Fort%20Player%20Pawn%20Object%20Reference) (Part :type Custom%20Character%20Part%20Object%20Reference) (MeshPart :type Skeletal%20Mesh%20Component%20Object%20Reference) (PartType :type EFortCustomPartType%20Enum :default "Head")))

(event EventOnWeaponChanged
  (params (NewWeapon :type Fort%20Weapon%20Object%20Reference) (OldWeapon :type Fort%20Weapon%20Object%20Reference))
  (bind _returnvalue (CustomCharacter|GetSkeletalMeshForPartOnPlayerPawn))
  (|CallParentFunction NewWeapon OldWeapon)
  (Utilities|IsValid NewWeapon
    (:Is%20Valid
      (Utilities|Casting|CastToB_DemonMask_Void_Teleport NewWeapon
        (:CastFailed
          (Utilities|IsValid _returnvalue
            (:Is%20Valid
              (Rendering|Material|SetCustomPrimitiveDataFloat _returnvalue 1 (Utilities|Time|GetGameTimeinSeconds)))
            (:Is%20Not%20Valid
              (Development|PrintString "INVALID"))))))))

(event EventOnCharacterCustomizationFinished
  (params (CharacterPartInfo :type Array%20of%20Applied%20Character%20Part%20Info%20Structures))
  (bind _returnvalue (CustomCharacter|GetSkeletalMeshForPartOnPlayerPawn))
  (|CallParentFunction CharacterPartInfo)
  (Utilities|IsValid _returnvalue
    (:Is%20Valid
      (Rendering|SetVisibility _returnvalue true true)
      (Rendering|Material|SetCustomPrimitiveDataFloat _returnvalue 0 (Utilities|Time|GetGameTimeinSeconds)))))

(event Custom|OnProjectileCooldownStarted
  ;[0.8698 0.3168 0.6393 0.5] Hand FX Cooldown reactivity
  (CallFunction|HandFxActivation false))

(event Custom|DemonMaskVoid_MaterialWipeOnToss
  (bind _returnvalue (CustomCharacter|GetSkeletalMeshForPartOnPlayerPawn))
  (bind _mask_equip_ns (Variables|Default|GetMask%20Equip%20NS))
  (Utilities|IsValid _returnvalue
    (:Is%20Valid
      (Rendering|Material|SetCustomPrimitiveDataFloat _returnvalue 1 (Utilities|Time|GetGameTimeinSeconds))
      (Utilities|FlowControl|Sequence
        (:then_0
          (Utilities|FlowControl|Delay 0.45)
          (Components|Activation|Deactivate _mask_equip_ns)
          (Rendering|SetVisibility _mask_equip_ns false true))
        (:then_1
          (Utilities|FlowControl|Delay 0.6)
          (Rendering|SetVisibility (CustomCharacter|GetSkeletalMeshForPartOnPlayerPawn) false false))))))

(event Custom|OnProjectileCooldownEnded
  (bind _ns_voidmask_handvoid (Variables|Default|GetNS_VoidMask_HandVoid))
  ;[0.8698 0.3168 0.6393 0.5] Hand FX Cooldown reactivity
  (CallFunction|HandFxActivation true)
  (if (Vehicle|IsinVehicle (CustomCharacter|GetAssociatedPlayerPawn))
    (Niagara|SetNiagaraVariable(Bool) _ns_voidmask_handvoid "HideOrb" true)
    (else
      (Niagara|SetNiagaraVariable(Bool) _ns_voidmask_handvoid "HideOrb" false))))

(event Custom|OnTeleportCooldownStarted
  ;[0.7094 0.3363 0.9844 0.5] Mask FX Cooldown reactivity
  (Niagara|SetNiagaraVariable(Bool) (Variables|Default|GetMask%20Equip%20NS) "AbilityOnCooldown" true))

(event EventOnHitPawn
  (params (SelfActor :type Actor%20Object%20Reference) (OtherActor :type Actor%20Object%20Reference) (NormalImpulse :type Vector :default "0, 0, 0") (Hit :type Hit%20Result%20Structure%20%28by%20ref%29)))

(event Custom|OnTeleportCooldownEnded
  ;[0.7094 0.3363 0.9844 0.5] Mask FX Cooldown reactivity
  (Niagara|SetNiagaraVariable(Bool) (Variables|Default|GetMask%20Equip%20NS) "AbilityOnCooldown" false))

(event EventOnEnterVehicle
  (params (VehicleEntered :type Fort%20Vehicle%20Interface%20Interface%20%28by%20ref%29))
  ;[0.8594 0.8023 0.2596 0.5] Deactivate Hand FX when entering vehicle
  (|CallParentFunction VehicleEntered)
  (Components|Activation|Deactivate (Variables|Default|GetNS_VoidMask_HandVoid)))

(event EventOnExitVehicle
  ;[0.8594 0.8023 0.2596 0.5] Deactivate Hand FX when entering vehicle
  (|CallParentFunction)
  (Components|Activation|Activate (Variables|Default|GetNS_VoidMask_HandVoid) false))

(event EventOnBeginEmote
  (params (MontageItemDef :type Fort%20Item%20Definition%20Object%20Reference) (PawnEmoting :type Fort%20Pawn%20Object%20Reference))
  ;[0.1381 0.5408 0.5877 0.5] Deactivate Hand FX when doing an Emote
  (|CallParentFunction MontageItemDef PawnEmoting)
  (Components|Activation|Deactivate (Variables|Default|GetNS_VoidMask_HandVoid)))

(event EventOnEndEmote
  (params (MontageItemDef :type Fort%20Item%20Definition%20Object%20Reference) (PawnEmoting :type Fort%20Pawn%20Object%20Reference))
  ;[0.1381 0.5408 0.5877 0.5] Deactivate Hand FX when doing an Emote
  (|CallParentFunction MontageItemDef PawnEmoting)
  (Components|Activation|Activate (Variables|Default|GetNS_VoidMask_HandVoid) false))

(fn HandFxActivation
  (params (Activate :type Boolean))
  (bind _ns_voidmask_handvoid (Variables|Default|GetNS_VoidMask_HandVoid))
  (if Activate
    (Components|Activation|Activate _ns_voidmask_handvoid false)
    (else
      (Components|Activation|Deactivate _ns_voidmask_handvoid))))

(fn ConstructionScript
  (|CallParentFunction))
